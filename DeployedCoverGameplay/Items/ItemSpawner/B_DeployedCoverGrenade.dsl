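; B_DeployedCoverGrenade is the deployed cover grenade item, a Blueprint built on the itemized prop spawner framework.
; On BeginPlay it caches native weapon data built by GetNativeData.
; That function combines prop transforms and chunk mapping from the data asset's prop list with the instigator pawn, trace ignore classes and ScalableFloat-evaluated targeting and placement tuning.
; Spawning logic is not in this graph, and Tick does no work here.
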(event EventBeginPlay
  (|CallParentFunction)
  (bind _outweapondata (Default|GetNativeData))
  (Variables|Default|SetNativeData _outweapondata))

(event EventTick
  (params (DeltaSeconds :type Float%20%28single-precision%29))
  (|CallParentFunction DeltaSeconds) ; This node is disabled and will not be called.\nDrag off pins to build functionality.
)

(fn GetNativeData
  (bind (_proprelativetransforms _chunklocationtopropindexmap) (ItemizedPropSpawner|BuildPropDataforWeaponData (Variables|/ItemizedPropSpawnerParent/Blueprints/DataClass_ItemizedPropSpawnerData.DataClass_ItemizedPropSpawnerData_C|GetNativePropsToSpawn (Variables|Default|GetData%20Asset))))
  (bind _asfort_player_pawn_athena (Utilities|Casting|CastToFortPlayerPawnAthena (Variables|Actor|GetInstigator)))
  (bind _output_get (LocalVariables|SetPlayer _asfort_player_pawn_athena))
  (bind _structout (Utilities|Struct|SetmembersinItemizedPropSpawnerWeaponData (LocalVariables|GetOutData) _output_get self (ScalableFloat|GetValueatLevel (Variables|/ItemizedPropSpawnerParent/Blueprints/DataClass_ItemizedPropSpawnerData.DataClass_ItemizedPropSpawnerData_C|GetMaxRange_ADSing (Variables|Default|GetData%20Asset))) (ScalableFloat|GetValueatLevel (Variables|/ItemizedPropSpawnerParent/Blueprints/DataClass_ItemizedPropSpawnerData.DataClass_ItemizedPropSpawnerData_C|GetMaxRange_Base (Variables|Default|GetData%20Asset))) (ScalableFloat|GetValueatLevel (Variables|/ItemizedPropSpawnerParent/Blueprints/DataClass_ItemizedPropSpawnerData.DataClass_ItemizedPropSpawnerData_C|GetMinDistance (Variables|Default|GetData%20Asset))) (ScalableFloat|GetValueatLevel (Variables|/ItemizedPropSpawnerParent/Blueprints/DataClass_ItemizedPropSpawnerData.DataClass_ItemizedPropSpawnerData_C|GetMinDistFromIndestructible (Variables|Default|GetData%20Asset))) (ScalableFloat|GetValueasBool (Variables|/ItemizedPropSpawnerParent/Blueprints/DataClass_ItemizedPropSpawnerData.DataClass_ItemizedPropSpawnerData_C|GetEnableBlockedByIndestructible (Variables|Default|GetData%20Asset))) (ScalableFloat|GetValueatLevel (Variables|/ItemizedPropSpawnerParent/Blueprints/DataClass_ItemizedPropSpawnerData.DataClass_ItemizedPropSpawnerData_C|GetPlayerForwardVectorScale (Variables|Default|GetData%20Asset))) (ScalableFloat|GetValueatLevel (Variables|/ItemizedPropSpawnerParent/Blueprints/DataClass_ItemizedPropSpawnerData.DataClass_ItemizedPropSpawnerData_C|GetInitialTracePullbackDist (Variables|Default|GetData%20Asset))) (ScalableFloat|GetValueatLevel (Variables|/ItemizedPropSpawnerParent/Blueprints/DataClass_ItemizedPropSpawnerData.DataClass_ItemizedPropSpawnerData_C|GetDownwardTraceDistanceNear (Variables|Default|GetData%20Asset))) (ScalableFloat|GetValueatLevel (Variables|/ItemizedPropSpawnerParent/Blueprints/DataClass_ItemizedPropSpawnerData.DataClass_ItemizedPropSpawnerData_C|GetDownwardTraceDistanceNearUp (Variables|Default|GetData%20Asset))) (ScalableFloat|GetValueatLevel (Variables|/ItemizedPropSpawnerParent/Blueprints/DataClass_ItemizedPropSpawnerData.DataClass_ItemizedPropSpawnerData_C|GetDownwardTraceDistanceFar (Variables|Default|GetData%20Asset))) (Variables|Default|GetClasses%20to%20Ignore%20when%20Tracing) (Variables|Default|GetClasses%20to%20Ignore%20when%20Tracing%20if%20Small) _proprelativetransforms _chunklocationtopropindexmap (ScalableFloat|GetValueatLevel (Variables|/ItemizedPropSpawnerParent/Blueprints/DataClass_ItemizedPropSpawnerData.DataClass_ItemizedPropSpawnerData_C|GetChunkRelativeRightVectorScale (Variables|Default|GetData%20Asset))) (ScalableFloat|GetValueatLevel (Variables|/ItemizedPropSpawnerParent/Blueprints/DataClass_ItemizedPropSpawnerData.DataClass_ItemizedPropSpawnerData_C|GetGroundFindStartOffset (Variables|Default|GetData%20Asset))) (ScalableFloat|GetValueatLevel (Variables|/ItemizedPropSpawnerParent/Blueprints/DataClass_ItemizedPropSpawnerData.DataClass_ItemizedPropSpawnerData_C|GetGroundFindEndOffset (Variables|Default|GetData%20Asset))) (ScalableFloat|GetValueatLevel (Variables|/ItemizedPropSpawnerParent/Blueprints/DataClass_ItemizedPropSpawnerData.DataClass_ItemizedPropSpawnerData_C|GetGroundFindDistanceOffset (Variables|Default|GetData%20Asset))) (ScalableFloat|GetValueatLevel (Variables|/ItemizedPropSpawnerParent/Blueprints/DataClass_ItemizedPropSpawnerData.DataClass_ItemizedPropSpawnerData_C|GetMaxSpawnDistBelowPlayer (Variables|Default|GetData%20Asset))) (Utilities|Array|MakeArray 0) (ScalableFloat|GetValueatLevel (Variables|/ItemizedPropSpawnerParent/Blueprints/DataClass_ItemizedPropSpawnerData.DataClass_ItemizedPropSpawnerData_C|GetIceBlockPitch (Variables|Default|GetData%20Asset))) (ScalableFloat|GetValueatLevel (Variables|/ItemizedPropSpawnerParent/Blueprints/DataClass_ItemizedPropSpawnerData.DataClass_ItemizedPropSpawnerData_C|GetBlockingIndestructibleObjectDist (Variables|Default|GetData%20Asset))) (ScalableFloat|GetValueatLevel (Variables|/ItemizedPropSpawnerParent/Blueprints/DataClass_ItemizedPropSpawnerData.DataClass_ItemizedPropSpawnerData_C|GetMaxVerticalChunkSpacing (Variables|Default|GetData%20Asset))) (ScalableFloat|GetValueatLevel (Variables|/ItemizedPropSpawnerParent/Blueprints/DataClass_ItemizedPropSpawnerData.DataClass_ItemizedPropSpawnerData_C|GetMinSupportingPropSize (Variables|Default|GetData%20Asset))) (ScalableFloat|GetValueatLevel (Variables|/ItemizedPropSpawnerParent/Blueprints/DataClass_ItemizedPropSpawnerData.DataClass_ItemizedPropSpawnerData_C|GetTooCloseForHeightDistStep (Variables|Default|GetData%20Asset))) (ScalableFloat|GetValueasInteger (Variables|/ItemizedPropSpawnerParent/Blueprints/DataClass_ItemizedPropSpawnerData.DataClass_ItemizedPropSpawnerData_C|GetTooCloseForHeightMaxIterations (Variables|Default|GetData%20Asset))) (ScalableFloat|GetValueatLevel (Variables|/ItemizedPropSpawnerParent/Blueprints/DataClass_ItemizedPropSpawnerData.DataClass_ItemizedPropSpawnerData_C|GetTooCloseForHeightMinHeightAbovePlayer (Variables|Default|GetData%20Asset))) (ScalableFloat|GetValueatLevel (Variables|/ItemizedPropSpawnerParent/Blueprints/DataClass_ItemizedPropSpawnerData.DataClass_ItemizedPropSpawnerData_C|GetTooCloseForHeightMaxHeightAbovePlayer (Variables|Default|GetData%20Asset))) (ScalableFloat|GetValueatLevel (Variables|/ItemizedPropSpawnerParent/Blueprints/DataClass_ItemizedPropSpawnerData.DataClass_ItemizedPropSpawnerData_C|GetTooCloseForHeightMinDistFromPlayer (Variables|Default|GetData%20Asset))) (ScalableFloat|GetValueatLevel (Variables|/ItemizedPropSpawnerParent/Blueprints/DataClass_ItemizedPropSpawnerData.DataClass_ItemizedPropSpawnerData_C|GetSphereTraceRadius (Variables|Default|GetData%20Asset))) (ScalableFloat|GetValueasBool (Variables|/ItemizedPropSpawnerParent/Blueprints/DataClass_ItemizedPropSpawnerData.DataClass_ItemizedPropSpawnerData_C|GetAllowUnwalkableTargets (Variables|Default|GetData%20Asset))) (ScalableFloat|GetValueasBool (Variables|/ItemizedPropSpawnerParent/Blueprints/DataClass_ItemizedPropSpawnerData.DataClass_ItemizedPropSpawnerData_C|GetAllowWaterTargets (Variables|Default|GetData%20Asset))) (ScalableFloat|GetValueatLevel (Variables|/ItemizedPropSpawnerParent/Blueprints/DataClass_ItemizedPropSpawnerData.DataClass_ItemizedPropSpawnerData_C|GetTryCloserOnInvalidTargetDist (Variables|Default|GetData%20Asset))) (ScalableFloat|GetValueatLevel (Variables|/ItemizedPropSpawnerParent/Blueprints/DataClass_ItemizedPropSpawnerData.DataClass_ItemizedPropSpawnerData_C|GetVerticalDistanceLimit (Variables|Default|GetData%20Asset)))))
  (return _structout))

(fn ConstructionScript
  (|CallParentFunction))
