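; Parent Blueprint for itemized deployed cover, handling extra damage from AI minions and removal when its supporting player build is edited.
; When hotfix-enabled, where cosmetic events cannot execute, damage from VividRazor minion-tagged pawns triggers an additional self-applied environmental hit scaled by a multiplier.
; On attaching to a player-buildable piece, it binds to that piece's replacement destruction and destroys itself when fired.
; Server death unbinds both listeners.
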(event EventBeginPlay
  (|CallParentFunction)
  (Networking|CanExecuteCosmeticEvents
    (:False
      (if (ScalableFloat|GetValueasBoolwithContext (Variables|Default|GetHF_EnableMinionDamageMultiplier))
        (EventDispatchers|BindEventtoOnDamaged (EventDispatchers|CreateEvent :!event Custom|WasDamaged))))))

(event EventOnDeathServer
  (params (Damage :type Float%20%28single-precision%29) (DamageTags :type Gameplay%20Tag%20Container%20Structure%20%28by%20ref%29) (Momentum :type Vector :default "0, 0, 0") (HitInfo :type Hit%20Result%20Structure%20%28by%20ref%29) (InstigatedBy :type Controller%20Object%20Reference) (DamageCauser :type Actor%20Object%20Reference) (EffectContext :type Gameplay%20Effect%20Context%20Handle%20Structure))
  (|CallParentFunction Damage DamageTags Momentum HitInfo InstigatedBy DamageCauser EffectContext)
  (Utilities|FlowControl|Sequence
    (:then_0
      (bind _pb_actorattachedto (Variables|Default|GetPB_ActorAttachedTo)
        (:then
          (Building|UnbindEventfromOnReplacementDestruction _pb_actorattachedto (EventDispatchers|CreateEvent :!event Custom|PBActorEdited)))
        (:else)))
    (:then_1
      (if (ScalableFloat|GetValueasBoolwithContext (Variables|Default|GetHF_EnableMinionDamageMultiplier))
        (EventDispatchers|UnbindEventfromOnDamaged (EventDispatchers|CreateEvent :!event Custom|WasDamaged))))))

(event Custom|WasDamaged
  (params (DamagedActor :type Actor%20Object%20Reference) (Damage :type Float%20%28single-precision%29) (InstigatedBy :type Controller%20Object%20Reference) (DamageCauser :type Actor%20Object%20Reference) (HitLocation :type Vector) (FHitComponent :type Primitive%20Component%20Object%20Reference) (BoneName :type Name) (Momentum :type Vector))
  (bind _returnvalue (Pawn|GetControlledPawn InstigatedBy))
  (bind _self self)
  ; Handle Vault AI dealing increased damaged to cover only if hotfixed on.
  (Utilities|IsValid InstigatedBy
    (:Is%20Valid
      (Utilities|IsValid _returnvalue
        (:Is%20Valid
          (bind _returnvalue_2 (GameplayTags|HasMatchingGameplayTag(Message) _returnvalue "(TagName=\"AISpawnerData.Type.VividRazor.Minions\")"))
          (if _returnvalue_2
            (Utilities|IsValid _self
              (:Is%20Valid
                (Game|Damage|ApplyDamage _self (* Damage (ScalableFloat|GetValueatLevel (Variables|Default|GetHF_MinionDamageMultiplier))) 0 _self "/Engine/EngineDamageTypes/DmgTypeBP_Environmental.DmgTypeBP_Environmental_C")))))))))

(event Attachment|EventOnAttachToBuilding
  (params (AttachedToBuilding :type Building%20Actor%20Object%20Reference))
  ; Destroy the cover piece if attached to an editable actor and it gets edited.
  (if (Variables|/Script/FortniteGame.BuildingActor|GetbIsPlayerBuildable AttachedToBuilding)
    (bind _asbuilding_static_mesh (Utilities|Casting|CastToBuildingSMActor AttachedToBuilding))
    (bind _output_get (Variables|Default|SetPB_ActorAttachedTo _asbuilding_static_mesh))
    (Building|BindEventtoOnReplacementDestruction _output_get (EventDispatchers|CreateEvent :!event Custom|PBActorEdited))))

(event Custom|PBActorEdited
  (params (ReplacementType :type EBuildingReplacementType%20Enum :default "BRT_None") (NewBuildingActor :type Building%20Static%20Mesh%20Object%20Reference))
  ; Destroy the cover piece if attached to an editable actor and it gets edited.
  (Actor|DestroyActor))

(fn ConstructionScript)
