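; Building Blueprint for the right side wall piece of itemized deployed cover, adding material setup on top of its parent class.
; BeginPlay calls the parent, then creates a dynamic instance of the right-side cover material on slot 0 of the static mesh.
; On placement, it calls the parent and writes the current game time into the BP_GetTime and BP_GetTime_Scale material parameters, giving the shader a placement timestamp.
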(event Building|EventOnBuildingPlaced
  (bind _returnvalue (Utilities|Time|GetGameTimeinSeconds))
  (bind _staticmeshcomponent (Variables|Debug|GetStaticMeshComponent))
  (|CallParentFunction)
  (Rendering|Material|SetScalarParameterValueonMaterials _staticmeshcomponent "BP_GetTime" _returnvalue)
  (Rendering|Material|SetScalarParameterValueonMaterials _staticmeshcomponent "BP_GetTime_Scale" _returnvalue))

(event EventBeginPlay
  (|CallParentFunction)
  (Rendering|Material|CreateDynamicMaterialInstance (Variables|Debug|GetStaticMeshComponent) 0 "/DeployedCoverGameplay/Assets/Materials/MI_Gameplay_DeployedCoverSide_R_A.MI_Gameplay_DeployedCoverSide_R_A" "None" :!class /Script/Engine.PrimitiveComponent))

(fn ConstructionScript)
