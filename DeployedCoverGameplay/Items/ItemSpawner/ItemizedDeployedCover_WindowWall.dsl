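; Deployable cover wall building actor that spawns and manages a separate window piece.
; On begin play it creates a dynamic center material; the server fires the landed GameplayCue and spawns the window as a building actor at a rotated offset, as dynamic build instructions require.
; Placement passes game time to material parameters.
; Server death flags the window destroyed once and expires it almost immediately.
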(event Building|EventOnBuildingPlaced
  (bind _returnvalue (Utilities|Time|GetGameTimeinSeconds))
  (bind _staticmeshcomponent (Variables|Debug|GetStaticMeshComponent))
  (|CallParentFunction)
  (Rendering|Material|SetScalarParameterValueonMaterials _staticmeshcomponent "BP_GetTime" _returnvalue)
  (Rendering|Material|SetScalarParameterValueonMaterials _staticmeshcomponent "BP_GetTime_Scale" _returnvalue))

(event EventBeginPlay
  (|CallParentFunction)
  (Rendering|Material|CreateDynamicMaterialInstance (Variables|Debug|GetStaticMeshComponent) 0 "/DeployedCoverGameplay/Assets/Materials/MI_Gameplay_DeployedCoverCenter_A.MI_Gameplay_DeployedCoverCenter_A" "None" :!class /Script/Engine.PrimitiveComponent)
  (Networking|SwitchHasAuthority
    (:Authority
      (GameplayCue|ExecuteGameplayCue self (Variables|Default|GetGCN_CoverLanded) (Utilities|Struct|MakeGameplayEffectContextHandle))
      (bind _returnvalue (Game|SpawnBuildingActorfromClass "/DeployedCoverGameplay/Items/DeployedActors/B_DeployedCover_Window.B_DeployedCover_Window_C" (Math|Transform|MakeTransform (+ (Math|Vector|RotateVector (Variables|Default|GetWindowSpawnLocation) (.rotation (Transformation|GetActorTransform))) (.location (Transformation|GetActorTransform))) (.rotation (Transformation|GetActorTransform)) (.scale (Transformation|GetActorTransform))) 0 false 0 0)) ; dynamic build instructions only work for Building Actors, not for other BP types\nspawn the window this way instead
      (Variables|Default|SetSpawnedWindow _returnvalue))))

(event EventOnDeathServer
  (params (Damage :type Float%20%28single-precision%29) (DamageTags :type Gameplay%20Tag%20Container%20Structure%20%28by%20ref%29) (Momentum :type Vector :default "0, 0, 0") (HitInfo :type Hit%20Result%20Structure%20%28by%20ref%29) (InstigatedBy :type Controller%20Object%20Reference) (DamageCauser :type Actor%20Object%20Reference) (EffectContext :type Gameplay%20Effect%20Context%20Handle%20Structure))
  (|CallParentFunction Damage DamageTags Momentum HitInfo InstigatedBy DamageCauser EffectContext)
  (bind _spawnedwindow (Variables|Default|GetSpawnedWindow)
    (:then
      (Utilities|FlowControl|DoOnce false
        (:Completed
          (Actor|SetLifeSpan _spawnedwindow 0.1)
          (Variables|/DeployedCoverGameplay/Items/DeployedActors/B_DeployedCover_ShieldedWindow.B_DeployedCover_ShieldedWindow_C|SetbHasBeenDestroyed true _spawnedwindow))))
    (:else)))

(fn ConstructionScript)
